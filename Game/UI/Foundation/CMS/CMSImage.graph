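# CMSImage is a reusable UI widget that wraps a single EpicCMSImage for showing CMS media.
# Its SetMediaURL function forwards a URL string to the inner Image, leaving any loading to that framework widget.
# The image does not match its size to the media, draws no loading background, and falls back to a transparent blank texture when loading fails.
# The event graph is empty.
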WidgetBlueprint /Game/UI/Foundation/CMS/CMSImage
graphs [
 EventGraph {links [], nodes []}
 SetMediaURL {nodes [
   n1 = SetMediaURL!() -> (then, MediaUrl:String) {is=title K2Node_FunctionEntry}
   > SetMediaURL!(execute, self:Epic CMSImage Object Reference=Get Image() -> (Image:Epic CMSImage Object Reference), MediaUrl:String=@n1) -> (then) {is=title K2Node_CallFunction}
  ]}
]
variables []
widgets {
 info {inheritedWidgetCount=0, parentClass=&UserWidget, rootWidgetClass=/Script/EpicCMSUIFramework.EpicCMSImage, widgetCount=1}
 widgets [Image {
   widgetClassPath=&EpicCMSImage
   bIsVariable=true
   widget=Image
   properties {
    bMatchImageSize=false
    brush {
     bIsDynamicallyLoaded=false
     drawAs=Image
     imageType=FullColor
     mirroring=NoMirror
     resourceName=None
     resourceObject=null
     tiling=NoTile
     imageSize=<1024 1024>
     margin=<ltrb 0 0 0 0>
     outlineSettings {
      bUseBrushTransparency=false
      roundingType=HalfHeightRadius
      width=0
      color {colorUseRule=UseColor_Specified, specifiedColor=<rgba 0 0 0 0>}
      cornerRadii=<0 0 0 0>
     }
     tintColor {colorUseRule=UseColor_Specified, specifiedColor=<rgba 1 1 1 1>}
     uVRegion {bIsValid=false, max=<0 0>, min=<0 0>}
    }
    loadingBackgroundBrush {
     bIsDynamicallyLoaded=false
     drawAs=NoDrawType
     imageType=NoImage
     mirroring=NoMirror
     resourceName=None
     resourceObject=null
     tiling=NoTile
     imageSize=<16 9>
     margin=<ltrb 0 0 0 0>
     outlineSettings {
      bUseBrushTransparency=false
      roundingType=HalfHeightRadius
      width=0
      color {colorUseRule=UseColor_Specified, specifiedColor=<rgba 0 0 0 0>}
      cornerRadii=<0 0 0 0>
     }
     tintColor {colorUseRule=UseColor_Specified, specifiedColor=<rgba 1 1 1 1>}
     uVRegion {bIsValid=false, max=<0 0>, min=<0 0>}
    }
    loadingFailFallback {
     bIsDynamicallyLoaded=false
     drawAs=Image
     imageType=NoImage
     mirroring=NoMirror
     resourceName=None
     resourceObject=/Game/Global/Textures/Default/Blanks/T_Blank_White_ZeroAlpha_sRGBColor
     tiling=NoTile
     imageSize=<1 1>
     margin=<ltrb 0 0 0 0>
     outlineSettings {
      bUseBrushTransparency=false
      roundingType=HalfHeightRadius
      width=0
      color {colorUseRule=UseColor_Specified, specifiedColor=<rgba 0 0 0 0>}
      cornerRadii=<0 0 0 0>
     }
     tintColor {colorUseRule=UseColor_Specified, specifiedColor=<rgba 1 1 1 1>}
     uVRegion {bIsValid=false, max=<0 0>, min=<0 0>}
    }
   }
  }]
}
